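# Physics asset for the belt of the M_MED_Skirmish_Alt male medium player body, modeling its left and right dangling straps.
# Each side has a three-capsule strap chain hung from a small base capsule, with larger pelvis and thigh capsules forming the surrounding body.
# Strap joints allow more swing on one axis than on the other or in twist, while base and thigh joints to the pelvis rotate freely.
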PhysicsAsset /BRCosmetics/Characters/Player/Male/Medium/Bodies/M_MED_Skirmish_Alt/Meshes/M_MED_Skirmish_Alt_Belt_Physics
bodies {
 dyn_strap_a_base_l [{shapeType=Capsule, length=0.505, radius=0.505, rotation=<rot 0 -90 0>}]
 dyn_strap_a_base_r [{shapeType=Capsule, length=0.505, radius=0.505, rotation=<rot 0 -90 0>}]
 dyn_strap_a_l [{shapeType=Capsule, length=3.25, radius=1.75, center=<3.107 -0.0001 0>, rotation=<rot 90 0 0>}]
 dyn_strap_a_r [{shapeType=Capsule, length=3, radius=2, center=<-3.0856 0 0>, rotation=<rot 90 0 0>}]
 dyn_strap_b_l [{shapeType=Capsule, length=3, radius=2.25, center=<2.9998 0 0>, rotation=<rot 90 0 0>}]
 dyn_strap_b_r [{shapeType=Capsule, length=2.75, radius=2.5, center=<-2.9443 0 0>, rotation=<rot 90 0 0>}]
 dyn_strap_c_l [{shapeType=Capsule, length=5, radius=2.25, center=<3.8646 0 0>, rotation=<rot 90 0 0>}]
 dyn_strap_c_r [{shapeType=Capsule, length=4.25, radius=2.5, center=<-3.7882 0 0>, rotation=<rot 90 0 0>}]
 pelvis [{shapeType=Capsule, length=12.5739, radius=13.1217, center=<-8.444 -1.5569 -0.2411>}]
 thigh_l [{shapeType=Capsule, length=25.75, radius=10.5, center=<-16.982 -0.0001 0>, rotation=<rot 90 -0.8952 -1.8873>}]
 thigh_r [{shapeType=Capsule, length=25.75, radius=10.5, center=<16.982 0.0001 0>, rotation=<rot 90 0.8952 179.9031>}]
}
constraints [
 {
  bone1Name=thigh_l
  bone2Name=pelvis
  swing1LimitDegrees=45
  swing1Motion=Free
  swing2LimitDegrees=45
  swing2Motion=Free
  twistLimitDegrees=45
  twistMotion=Free
 }
 {
  bone1Name=thigh_r
  bone2Name=pelvis
  swing1LimitDegrees=45
  swing1Motion=Free
  swing2LimitDegrees=45
  swing2Motion=Free
  twistLimitDegrees=45
  twistMotion=Free
 }
 {bone1Name=dyn_strap_a_l, bone2Name=dyn_strap_a_base_l, swing1LimitDegrees=30, swing2LimitDegrees=5, twistLimitDegrees=5}
 {bone1Name=dyn_strap_b_l, bone2Name=dyn_strap_a_l, swing1LimitDegrees=30, swing2LimitDegrees=5, twistLimitDegrees=5}
 {bone1Name=dyn_strap_c_l, bone2Name=dyn_strap_b_l, swing1LimitDegrees=30, swing2LimitDegrees=5, twistLimitDegrees=5}
 {bone1Name=dyn_strap_a_r, bone2Name=dyn_strap_a_base_r, swing1LimitDegrees=30, swing2LimitDegrees=5, twistLimitDegrees=5}
 {bone1Name=dyn_strap_b_r, bone2Name=dyn_strap_a_r, swing1LimitDegrees=30, swing2LimitDegrees=5, twistLimitDegrees=5}
 {bone1Name=dyn_strap_c_r, bone2Name=dyn_strap_b_r, swing1LimitDegrees=30, swing2LimitDegrees=5, twistLimitDegrees=5}
 {
  bone1Name=dyn_strap_a_base_l
  bone2Name=pelvis
  swing1LimitDegrees=45
  swing1Motion=Free
  swing2LimitDegrees=45
  swing2Motion=Free
  twistLimitDegrees=45
  twistMotion=Free
 }
 {
  bone1Name=dyn_strap_a_base_r
  bone2Name=pelvis
  swing1LimitDegrees=45
  swing1Motion=Free
  swing2LimitDegrees=45
  swing2Motion=Free
  twistLimitDegrees=45
  twistMotion=Free
 }
]
