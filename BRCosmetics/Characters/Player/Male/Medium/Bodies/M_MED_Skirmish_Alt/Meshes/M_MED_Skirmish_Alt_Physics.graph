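# Physics asset for the male medium Skirmish Alt body cosmetic, covering its dangling left-arm straps and left and right pants scraps.
# Capsule bodies cover these dyn_ bones plus the left upper arm and forearm.
# Two strap chains use tightly limited links, with their bases and the forearm joined to the upper arm by angularly free constraints.
# Each pants scrap hinges on its base along one limited swing axis, mirrored per side.
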PhysicsAsset /BRCosmetics/Characters/Player/Male/Medium/Bodies/M_MED_Skirmish_Alt/Meshes/M_MED_Skirmish_Alt_Physics
bodies {
 dyn_arm_strap_aa_base_l [{shapeType=Capsule, length=0.505, radius=0.505, rotation=<rot 0 -90 0>}]
 dyn_arm_strap_aa_l [{shapeType=Capsule, length=2.75, radius=1.5, center=<2 0 0>, rotation=<rot 90 0 0>}]
 dyn_arm_strap_ab_l [{shapeType=Capsule, length=2.5, radius=1.75, center=<2 0 0>, rotation=<rot 90 0 0>}]
 dyn_arm_strap_ba_base_l [{shapeType=Capsule, length=0.505, radius=0.505, rotation=<rot 0 -90 0>}]
 dyn_arm_strap_ba_l [{shapeType=Capsule, length=2.75, radius=1.5, center=<2 0 0>, rotation=<rot 90 0 0>}]
 dyn_arm_strap_bb_l [{shapeType=Capsule, length=2.75, radius=1.5, center=<2 0 0>, rotation=<rot 90 0 0>}]
 dyn_pants_scrap_base_l [{shapeType=Capsule, length=0.505, radius=0.505, rotation=<rot 0 -90 0>}]
 dyn_pants_scrap_base_r [{shapeType=Capsule, length=0.505, radius=0.505, rotation=<rot 0 -90 0>}]
 dyn_pants_scrap_l [{shapeType=Capsule, length=2.5, radius=2.5, center=<2.5 0 0>, rotation=<rot 90 0 0>}]
 dyn_pants_scrap_r [{shapeType=Capsule, length=2.5, radius=2.5, center=<2.5 0 0>, rotation=<rot 90 0 0>}]
 lowerarm_l [{shapeType=Capsule, length=7.75, radius=9, center=<7.0643 0 -0.5403>, rotation=<rot 85.2656 -0.0001 0>}]
 upperarm_l [{shapeType=Capsule, length=12.25, radius=7.5, center=<16.008 0.0002 -1.524>, rotation=<rot 90 0 0>}]
}
constraints [
 {bone1Name=dyn_arm_strap_aa_l, bone2Name=dyn_arm_strap_aa_base_l, swing1LimitDegrees=3, swing2LimitDegrees=3, twistLimitDegrees=3}
 {bone1Name=dyn_arm_strap_ab_l, bone2Name=dyn_arm_strap_aa_l, swing1LimitDegrees=3, swing2LimitDegrees=3, twistLimitDegrees=3}
 {bone1Name=dyn_arm_strap_ba_l, bone2Name=dyn_arm_strap_ba_base_l, swing1LimitDegrees=3, swing2LimitDegrees=3, twistLimitDegrees=3}
 {bone1Name=dyn_arm_strap_bb_l, bone2Name=dyn_arm_strap_ba_l, swing1LimitDegrees=3, swing2LimitDegrees=3, twistLimitDegrees=3}
 {
  bone1Name=dyn_arm_strap_aa_base_l
  bone2Name=upperarm_l
  swing1LimitDegrees=45
  swing1Motion=Free
  swing2LimitDegrees=45
  swing2Motion=Free
  twistLimitDegrees=45
  twistMotion=Free
 }
 {
  bone1Name=dyn_arm_strap_ba_base_l
  bone2Name=upperarm_l
  swing1LimitDegrees=45
  swing1Motion=Free
  swing2LimitDegrees=45
  swing2Motion=Free
  twistLimitDegrees=45
  twistMotion=Free
 }
 {
  bone1Name=lowerarm_l
  bone2Name=upperarm_l
  swing1LimitDegrees=45
  swing1Motion=Free
  swing2LimitDegrees=45
  swing2Motion=Free
  twistLimitDegrees=45
  twistMotion=Free
 }
 {
  bone1Name=dyn_pants_scrap_l
  bone2Name=dyn_pants_scrap_base_l
  swing1LimitDegrees=15
  swing2LimitDegrees=5
  swing2Motion=Locked
  twistLimitDegrees=5
  twistMotion=Locked
 }
 {
  bone1Name=dyn_pants_scrap_r
  bone2Name=dyn_pants_scrap_base_r
  swing1LimitDegrees=5
  swing1Motion=Locked
  swing2LimitDegrees=15
  twistLimitDegrees=5
  twistMotion=Locked
 }
]
