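# Physics asset for the male medium Skirmish Alt body, covering the lower legs and dynamic trouser-bottom bones, apparently for secondary cloth-like motion at the cuffs.
# Capsules on calves and feet anchor two-bone chains of tiny base and larger tip capsules, eight left and six right.
# Foot-to-calf and base-to-foot joints rotate freely, while each tip has narrow swing and tighter twist limits against its base.
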PhysicsAsset /Game/Characters/Player/Male/Medium/Bodies/M_MED_Skirmish_Alt/Meshes/M_MED_Skirmish_Shared_Physics
bodies {
 calf_l [{shapeType=Capsule, length=10.25, radius=5.75, center=<-44.2397 -0.1511 0.4683>, rotation=<rot 90 0 0>}]
 calf_r [{shapeType=Capsule, length=10.25, radius=5.75, center=<44.2397 0.1511 -0.4683>, rotation=<rot 90 0 180>}]
 dyn_pant_bttm_a_base_l [{shapeType=Capsule, length=0.505, radius=0.505, rotation=<rot 0 -90 0>}]
 dyn_pant_bttm_a_base_r [{shapeType=Capsule, length=0.505, radius=0.505, rotation=<rot 0 -90 0>}]
 dyn_pant_bttm_a_l [{shapeType=Capsule, length=3, radius=1.5, center=<2 0 0>, rotation=<rot 90 0 0>}]
 dyn_pant_bttm_a_r [{shapeType=Capsule, length=3.5, radius=1.75, center=<3 0 0>, rotation=<rot 90 0 0>}]
 dyn_pant_bttm_b_base_l [{shapeType=Capsule, length=0.505, radius=0.505, rotation=<rot 0 -90 0>}]
 dyn_pant_bttm_b_base_r [{shapeType=Capsule, length=0.505, radius=0.505, rotation=<rot 0 -90 0>}]
 dyn_pant_bttm_b_l [{shapeType=Capsule, length=3, radius=1.5, center=<2 0 0>, rotation=<rot 90 0 0>}]
 dyn_pant_bttm_b_r [{shapeType=Capsule, length=4, radius=1.5, center=<3 0 0>, rotation=<rot 90 0 0>}]
 dyn_pant_bttm_c_base_l [{shapeType=Capsule, length=0.505, radius=0.505, rotation=<rot 0 -90 0>}]
 dyn_pant_bttm_c_base_r [{shapeType=Capsule, length=0.505, radius=0.505, rotation=<rot 0 -90 0>}]
 dyn_pant_bttm_c_l [{shapeType=Capsule, length=3, radius=1.5, center=<2.8054 0 0>, rotation=<rot 90 0 0>}]
 dyn_pant_bttm_c_r [{shapeType=Capsule, length=4.5, radius=2, center=<3.736 0 0>, rotation=<rot 90 0 0>}]
 dyn_pant_bttm_d_base_l [{shapeType=Capsule, length=0.505, radius=0.505, rotation=<rot 0 -90 0>}]
 dyn_pant_bttm_d_base_r [{shapeType=Capsule, length=0.505, radius=0.505, rotation=<rot 0 -90 0>}]
 dyn_pant_bttm_d_l [{shapeType=Capsule, length=3.5, radius=1.5, center=<3.4833 0 0>, rotation=<rot 90 0 0>}]
 dyn_pant_bttm_d_r [{shapeType=Capsule, length=3.5, radius=1.75, center=<3 0 0>, rotation=<rot 90 0 0>}]
 dyn_pant_bttm_e_base_l [{shapeType=Capsule, length=0.505, radius=0.505, rotation=<rot 0 -90 0>}]
 dyn_pant_bttm_e_base_r [{shapeType=Capsule, length=0.505, radius=0.505, rotation=<rot 0 -90 0>}]
 dyn_pant_bttm_e_l [{shapeType=Capsule, length=3, radius=1.5, center=<2 0 0>, rotation=<rot 90 0 0>}]
 dyn_pant_bttm_e_r [{shapeType=Capsule, length=3.75, radius=1.75, center=<3.2198 0 0>, rotation=<rot 90 0 0>}]
 dyn_pant_bttm_f_base_l [{shapeType=Capsule, length=0.505, radius=0.505, rotation=<rot 0 -90 0>}]
 dyn_pant_bttm_f_base_r [{shapeType=Capsule, length=0.505, radius=0.505, rotation=<rot 0 -90 0>}]
 dyn_pant_bttm_f_l [{shapeType=Capsule, length=4, radius=1.5, center=<2.9729 0 0>, rotation=<rot 90 0 0>}]
 dyn_pant_bttm_f_r [{shapeType=Capsule, length=3.75, radius=1.5, center=<3 0 0>, rotation=<rot 90 0 0>}]
 dyn_pant_bttm_g_base_l [{shapeType=Capsule, length=0.505, radius=0.505, rotation=<rot 0 -90 0>}]
 dyn_pant_bttm_g_l [{shapeType=Capsule, length=3, radius=1.5, center=<2 0 0>, rotation=<rot 90 0 0>}]
 dyn_pant_bttm_h_base_l [{shapeType=Capsule, length=0.505, radius=0.505, rotation=<rot 0 -90 0>}]
 dyn_pant_bttm_h_l [{shapeType=Capsule, length=3, radius=1.5, center=<2 0 0>, rotation=<rot 90 0 0>}]
 foot_l [{shapeType=Capsule, length=7.9124, radius=5.1601, center=<-3.9165 4.3209 -0.2568>, rotation=<rot -2.3172 31.9187 -84.6954>}]
 foot_r [{shapeType=Capsule, length=7.9124, radius=5.1601, center=<3.9165 -4.3209 0.2568>, rotation=<rot -2.3172 31.9187 95.3046>}]
}
constraints [
 {
  bone1Name=foot_l
  bone2Name=calf_l
  swing1LimitDegrees=45
  swing1Motion=Free
  swing2LimitDegrees=45
  swing2Motion=Free
  twistLimitDegrees=45
  twistMotion=Free
 }
 {
  bone1Name=foot_r
  bone2Name=calf_r
  swing1LimitDegrees=45
  swing1Motion=Free
  swing2LimitDegrees=45
  swing2Motion=Free
  twistLimitDegrees=45
  twistMotion=Free
 }
 {
  bone1Name=dyn_pant_bttm_a_base_l
  bone2Name=foot_l
  swing1LimitDegrees=45
  swing1Motion=Free
  swing2LimitDegrees=45
  swing2Motion=Free
  twistLimitDegrees=45
  twistMotion=Free
 }
 {
  bone1Name=dyn_pant_bttm_f_base_l
  bone2Name=foot_l
  swing1LimitDegrees=45
  swing1Motion=Free
  swing2LimitDegrees=45
  swing2Motion=Free
  twistLimitDegrees=45
  twistMotion=Free
 }
 {
  bone1Name=dyn_pant_bttm_d_base_l
  bone2Name=foot_l
  swing1LimitDegrees=45
  swing1Motion=Free
  swing2LimitDegrees=45
  swing2Motion=Free
  twistLimitDegrees=45
  twistMotion=Free
 }
 {
  bone1Name=dyn_pant_bttm_c_base_l
  bone2Name=foot_l
  swing1LimitDegrees=45
  swing1Motion=Free
  swing2LimitDegrees=45
  swing2Motion=Free
  twistLimitDegrees=45
  twistMotion=Free
 }
 {
  bone1Name=dyn_pant_bttm_h_base_l
  bone2Name=foot_l
  swing1LimitDegrees=45
  swing1Motion=Free
  swing2LimitDegrees=45
  swing2Motion=Free
  twistLimitDegrees=45
  twistMotion=Free
 }
 {
  bone1Name=dyn_pant_bttm_b_base_l
  bone2Name=foot_l
  swing1LimitDegrees=45
  swing1Motion=Free
  swing2LimitDegrees=45
  swing2Motion=Free
  twistLimitDegrees=45
  twistMotion=Free
 }
 {
  bone1Name=dyn_pant_bttm_g_base_l
  bone2Name=foot_l
  swing1LimitDegrees=45
  swing1Motion=Free
  swing2LimitDegrees=45
  swing2Motion=Free
  twistLimitDegrees=45
  twistMotion=Free
 }
 {
  bone1Name=dyn_pant_bttm_e_base_l
  bone2Name=foot_l
  swing1LimitDegrees=45
  swing1Motion=Free
  swing2LimitDegrees=45
  swing2Motion=Free
  twistLimitDegrees=45
  twistMotion=Free
 }
 {bone1Name=dyn_pant_bttm_a_l, bone2Name=dyn_pant_bttm_a_base_l, swing1LimitDegrees=15, swing2LimitDegrees=15, twistLimitDegrees=5}
 {bone1Name=dyn_pant_bttm_f_l, bone2Name=dyn_pant_bttm_f_base_l, swing1LimitDegrees=15, swing2LimitDegrees=15, twistLimitDegrees=5}
 {bone1Name=dyn_pant_bttm_d_l, bone2Name=dyn_pant_bttm_d_base_l, swing1LimitDegrees=15, swing2LimitDegrees=15, twistLimitDegrees=5}
 {bone1Name=dyn_pant_bttm_c_l, bone2Name=dyn_pant_bttm_c_base_l, swing1LimitDegrees=15, swing2LimitDegrees=15, twistLimitDegrees=5}
 {bone1Name=dyn_pant_bttm_h_l, bone2Name=dyn_pant_bttm_h_base_l, swing1LimitDegrees=15, swing2LimitDegrees=15, twistLimitDegrees=5}
 {bone1Name=dyn_pant_bttm_b_l, bone2Name=dyn_pant_bttm_b_base_l, swing1LimitDegrees=15, swing2LimitDegrees=15, twistLimitDegrees=5}
 {bone1Name=dyn_pant_bttm_g_l, bone2Name=dyn_pant_bttm_g_base_l, swing1LimitDegrees=15, swing2LimitDegrees=15, twistLimitDegrees=5}
 {bone1Name=dyn_pant_bttm_e_l, bone2Name=dyn_pant_bttm_e_base_l, swing1LimitDegrees=15, swing2LimitDegrees=15, twistLimitDegrees=5}
 {
  bone1Name=dyn_pant_bttm_a_base_r
  bone2Name=foot_r
  swing1LimitDegrees=45
  swing1Motion=Free
  swing2LimitDegrees=45
  swing2Motion=Free
  twistLimitDegrees=45
  twistMotion=Free
 }
 {
  bone1Name=dyn_pant_bttm_b_base_r
  bone2Name=foot_r
  swing1LimitDegrees=45
  swing1Motion=Free
  swing2LimitDegrees=45
  swing2Motion=Free
  twistLimitDegrees=45
  twistMotion=Free
 }
 {
  bone1Name=dyn_pant_bttm_c_base_r
  bone2Name=foot_r
  swing1LimitDegrees=45
  swing1Motion=Free
  swing2LimitDegrees=45
  swing2Motion=Free
  twistLimitDegrees=45
  twistMotion=Free
 }
 {
  bone1Name=dyn_pant_bttm_d_base_r
  bone2Name=foot_r
  swing1LimitDegrees=45
  swing1Motion=Free
  swing2LimitDegrees=45
  swing2Motion=Free
  twistLimitDegrees=45
  twistMotion=Free
 }
 {
  bone1Name=dyn_pant_bttm_e_base_r
  bone2Name=foot_r
  swing1LimitDegrees=45
  swing1Motion=Free
  swing2LimitDegrees=45
  swing2Motion=Free
  twistLimitDegrees=45
  twistMotion=Free
 }
 {
  bone1Name=dyn_pant_bttm_f_base_r
  bone2Name=foot_r
  swing1LimitDegrees=45
  swing1Motion=Free
  swing2LimitDegrees=45
  swing2Motion=Free
  twistLimitDegrees=45
  twistMotion=Free
 }
 {bone1Name=dyn_pant_bttm_a_r, bone2Name=dyn_pant_bttm_a_base_r, swing1LimitDegrees=15, swing2LimitDegrees=15, twistLimitDegrees=5}
 {bone1Name=dyn_pant_bttm_b_r, bone2Name=dyn_pant_bttm_b_base_r, swing1LimitDegrees=15, swing2LimitDegrees=15, twistLimitDegrees=5}
 {bone1Name=dyn_pant_bttm_c_r, bone2Name=dyn_pant_bttm_c_base_r, swing1LimitDegrees=15, swing2LimitDegrees=15, twistLimitDegrees=5}
 {bone1Name=dyn_pant_bttm_d_r, bone2Name=dyn_pant_bttm_d_base_r, swing1LimitDegrees=15, swing2LimitDegrees=15, twistLimitDegrees=5}
 {bone1Name=dyn_pant_bttm_e_r, bone2Name=dyn_pant_bttm_e_base_r, swing1LimitDegrees=15, swing2LimitDegrees=15, twistLimitDegrees=5}
 {bone1Name=dyn_pant_bttm_f_r, bone2Name=dyn_pant_bttm_f_base_r, swing1LimitDegrees=15, swing2LimitDegrees=15, twistLimitDegrees=5}
]
