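# Animation Blueprint for the M_MED_Skirmish_Alt player body, adding secondary deformation and cloth motion to the pose copied from its parent mesh.
# Constraints and alpha-mapped additive bone edits correct limbs, shoulders, packs and straps, ending in three RigidBody simulations based on dyn_root_simspace.
# Update logic derives alphas from joint deltas, crouch, front-end and pistol-aim state, plus noisy wave rotators for straps.
# Sim-space rotation and gravity override use functions marked DEPRECATED.
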AnimBlueprint /BRCosmetics/Characters/Player/Male/Medium/Bodies/M_MED_Skirmish_Alt/Meshes/M_MED_Skirmish_Alt_AnimBP
target_skeleton=&/M_MED_Skirmish_Alt_Skeleton
graphs [
 AnimGraph {nodes [
   n1 = "Constraint - dyn_volume_base_lowerarm_l (PP)"!(
    ComponentPose=Local To Component(LocalPose=Copy Pose From Mesh() {bUseAttachedParent=true})
    Alpha=1
    ConstraintWeights_0=0.65
    ConstraintWeights_1:float=0.35
   ) -> (Pose) {
    alphaBoolBlend {blendOption=Linear}
    boneToModify {boneName=dyn_volume_base_lowerarm_l}
    constraintSetup [{targetBone {boneName=upperarm_l}}, {targetBone {boneName=lowerarm_l}}]
    constraintWeights [1,1]
   }
   n2 = "Constraint - dyn_volume_base_lowerarm_r (PP)"!(ComponentPose=@n1, Alpha=1, ConstraintWeights_0=0.65, ConstraintWeights_1:float=0.35) -> (Pose) {
    alphaBoolBlend {blendOption=Linear}
    boneToModify {boneName=dyn_volume_base_lowerarm_r}
    constraintSetup [{targetBone {boneName=upperarm_r}}, {targetBone {boneName=lowerarm_r}}]
    constraintWeights [1,1]
   }
   n3 = "Transform (Modify) Bone - Bone: dyn_volume_lowerarm_l"(ComponentPose=@n2, Alpha=Get elbow_rot_l() -> (elbow_rot_l:double), Translation="0,-3,0", Rotation="0,0,0", Scale="0.3,1,1") {
    rotationMode=BMM_Additive
    rotationSpace=BCS_BoneSpace
    scaleMode=BMM_Additive
    scaleSpace=BCS_BoneSpace
    translationMode=BMM_Additive
    translationSpace=BCS_BoneSpace
    alphaBoolBlend {blendOption=Linear}
    alphaScaleBiasClamp {bMapRange=true, inRange {max=80}}
    boneToModify {boneName=dyn_volume_lowerarm_l}
    scale {x=0.3}
    translation {y=-3}
   }
   n4 = "Transform (Modify) Bone - Bone: dyn_volume_lowerarm_r"(ComponentPose=@n3, Alpha=Get elbow_rot_r() -> (elbow_rot_r:double), Translation="0,3,0", Rotation="0,0,0", Scale="0.3,1,1") {
    rotationMode=BMM_Additive
    rotationSpace=BCS_BoneSpace
    scaleMode=BMM_Additive
    scaleSpace=BCS_BoneSpace
    translationMode=BMM_Additive
    translationSpace=BCS_BoneSpace
    alphaBoolBlend {blendOption=Linear}
    alphaScaleBiasClamp {bMapRange=true, inRange {max=80}}
    boneToModify {boneName=dyn_volume_lowerarm_r}
    scale {x=0.3}
    translation {y=3}
   }
   n5 = "Constraint - dyn_volume_base_thigh_l (PP)"!(ComponentPose=@n4, Alpha=1, ConstraintWeights_0=0.6, ConstraintWeights_1:float=0.4) -> (Pose) {
    alphaBoolBlend {blendOption=Linear}
    boneToModify {boneName=dyn_volume_base_thigh_l}
    constraintSetup [{targetBone {boneName=pelvis}}, {targetBone {boneName=thigh_l}}]
    constraintWeights [1,1]
   }
   n6 = "Constraint - dyn_volume_base_thigh_r (PP)"!(ComponentPose=@n5, Alpha=1, ConstraintWeights_0=0.6, ConstraintWeights_1:float=0.4) -> (Pose) {
    alphaBoolBlend {blendOption=Linear}
    boneToModify {boneName=dyn_volume_base_thigh_r}
    constraintSetup [{targetBone {boneName=pelvis}}, {targetBone {boneName=thigh_r}}]
    constraintWeights [1,1]
   }
   n7 = Get thigh_rot_l()
   n8 = "Transform (Modify) Bone - Bone: dyn_volume_thigh_l"(ComponentPose=@n6, Alpha=@n7, Translation="-10,-5,0", Rotation="0,0,0", Scale="1,1,1") {
    rotationMode=BMM_Additive
    rotationSpace=BCS_BoneSpace
    scaleMode=BMM_Additive
    scaleSpace=BCS_BoneSpace
    translationMode=BMM_Additive
    translationSpace=BCS_BoneSpace
    alphaBoolBlend {blendOption=Linear}
    alphaScaleBiasClamp {bMapRange=true, inRange {max=0, min=-100}, outRange {max=0, min=1}}
    boneToModify {boneName=dyn_volume_thigh_l}
    translation=<-10 -5>
   }
   n9 = "Transform (Modify) Bone - Bone: dyn_volume_thigh_l"(ComponentPose=@n8, Alpha=@n7, Translation="0,-4,0", Rotation="0,0,0", Scale="1,1,1") {
    rotationMode=BMM_Additive
    rotationSpace=BCS_BoneSpace
    scaleMode=BMM_Additive
    scaleSpace=BCS_BoneSpace
    translationMode=BMM_Additive
    translationSpace=BCS_BoneSpace
    alphaBoolBlend {blendOption=Linear}
    alphaScaleBiasClamp {bMapRange=true, inRange {max=50}}
    boneToModify {boneName=dyn_volume_thigh_l}
    translation {y=-4}
   }
   n10 = Get thigh_rot_r()
   n11 = "Transform (Modify) Bone - Bone: dyn_volume_thigh_r"(ComponentPose=@n9, Alpha=@n10, Translation="10,2,0", Rotation="0,0,0", Scale="1,1,1") {
    rotationMode=BMM_Additive
    rotationSpace=BCS_BoneSpace
    scaleMode=BMM_Additive
    scaleSpace=BCS_BoneSpace
    translationMode=BMM_Additive
    translationSpace=BCS_BoneSpace
    alphaBoolBlend {blendOption=Linear}
    alphaScaleBiasClamp {bMapRange=true, inRange {max=0, min=-100}, outRange {max=0, min=1}}
    boneToModify {boneName=dyn_volume_thigh_r}
    translation=<10 2>
   }
   n12 = "Transform (Modify) Bone - Bone: dyn_volume_thigh_r"(ComponentPose=@n11, Alpha=@n10, Translation="0,4,0", Rotation="0,0,0", Scale="1,1,1") {
    rotationMode=BMM_Additive
    rotationSpace=BCS_BoneSpace
    scaleMode=BMM_Additive
    scaleSpace=BCS_BoneSpace
    translationMode=BMM_Additive
    translationSpace=BCS_BoneSpace
    alphaBoolBlend {blendOption=Linear}
    alphaScaleBiasClamp {bMapRange=true, inRange {max=50}}
    boneToModify {boneName=dyn_volume_thigh_r}
    translation {y=4}
   }
   n13 = "Constraint - dyn_volume_base_calf_l (PP)"!(ComponentPose=@n12, Alpha=1, ConstraintWeights_0=0.5, ConstraintWeights_1:float=0.5) -> (Pose) {
    alphaBoolBlend {blendOption=Linear}
    boneToModify {boneName=dyn_volume_base_calf_l}
    constraintSetup [{targetBone {boneName=thigh_l}}, {targetBone {boneName=calf_l}}]
    constraintWeights [1,1]
   }
   n14 = "Constraint - dyn_volume_base_calf_r (PP)"!(ComponentPose=@n13, Alpha=1, ConstraintWeights_0=0.5, ConstraintWeights_1:float=0.5) -> (Pose) {
    alphaBoolBlend {blendOption=Linear}
    boneToModify {boneName=dyn_volume_base_calf_r}
    constraintSetup [{targetBone {boneName=thigh_r}}, {targetBone {boneName=calf_r}}]
    constraintWeights [1,1]
   }
   n15 = "Transform (Modify) Bone - Bone: dyn_volume_calf_l"(ComponentPose=@n14, Alpha=Get knee_rot_l() -> (knee_rot_l:double), Translation="3,3,0", Rotation="0,0,0", Scale="1,1,1.5") {
    rotationMode=BMM_Additive
    rotationSpace=BCS_BoneSpace
    scaleMode=BMM_Additive
    scaleSpace=BCS_BoneSpace
    translationMode=BMM_Additive
    translationSpace=BCS_BoneSpace
    alphaBoolBlend {blendOption=Linear}
    alphaScaleBiasClamp {bMapRange=true, inRange {max=80}}
    boneToModify {boneName=dyn_volume_calf_l}
    scale {z=1.5}
    translation=<3 3>
   }
   n16 = "Transform (Modify) Bone - Bone: dyn_volume_calf_r"(ComponentPose=@n15, Alpha=Get knee_rot_r() -> (knee_rot_r:double), Translation="-3,-3,0", Rotation="0,0,0", Scale="1,1,1.5") {
    rotationMode=BMM_Additive
    rotationSpace=BCS_BoneSpace
    scaleMode=BMM_Additive
    scaleSpace=BCS_BoneSpace
    translationMode=BMM_Additive
    translationSpace=BCS_BoneSpace
    alphaBoolBlend {blendOption=Linear}
    alphaScaleBiasClamp {bMapRange=true, inRange {max=80}}
    boneToModify {boneName=dyn_volume_calf_r}
    scale {z=1.5}
    translation=<-3 -3>
   }
   n17 = "Constraint - dyn_pants_scrap_base_r (PPP)"!(ComponentPose=@n16, Alpha=1, ConstraintWeights_0=0.1655, ConstraintWeights_1:float=0.0432, ConstraintWeights_2:float=0.7913) -> (Pose) {
    alphaBoolBlend {blendOption=Linear}
    boneToModify {boneName=dyn_pants_scrap_base_r}
    constraintSetup [{targetBone {boneName=thigh_r}}, {targetBone {boneName=pelvis}}, {targetBone {boneName=thigh_twist_01_r}}]
    constraintWeights [1,1,1]
   }
   n18 = "Constraint - dyn_pants_scrap_base_l (PPP)"!(ComponentPose=@n17, Alpha=1, ConstraintWeights_0=0.29, ConstraintWeights_1:float=0.0371, ConstraintWeights_2:float=0.6729) -> (Pose) {
    alphaBoolBlend {blendOption=Linear}
    boneToModify {boneName=dyn_pants_scrap_base_l}
    constraintSetup [{targetBone {boneName=calf_l}}, {targetBone {boneName=calf_twist_01_l}}, {targetBone {boneName=calf_twist_02_l}}]
    constraintWeights [1,1,1]
   }
   n19 = "Constraint - dyn_arm_strap_ba_base_l (PPP)"!(ComponentPose=@n18, Alpha=1, ConstraintWeights_0=0.3692, ConstraintWeights_1:float=0.3358, ConstraintWeights_2:float=0.295) -> (Pose) {
    alphaBoolBlend {blendOption=Linear}
    boneToModify {boneName=dyn_arm_strap_ba_base_l}
    constraintSetup [{targetBone {boneName=upperarm_twist_02_l}}, {targetBone {boneName=upperarm_l}}, {targetBone {boneName=upperarm_twist_01_l}}]
    constraintWeights [1,1,1]
   }
   n20 = "Constraint - dyn_arm_strap_aa_base_l (PPP)"!(ComponentPose=@n19, Alpha=1, ConstraintWeights_0=0.3658, ConstraintWeights_1:float=0.4121, ConstraintWeights_2:float=0.2221) -> (Pose) {
    alphaBoolBlend {blendOption=Linear}
    boneToModify {boneName=dyn_arm_strap_aa_base_l}
    constraintSetup [{targetBone {boneName=upperarm_twist_02_l}}, {targetBone {boneName=upperarm_l}}, {targetBone {boneName=upperarm_twist_01_l}}]
    constraintWeights [1,1,1]
   }
   n21 = "Constraint - dyn_pant_bttm_a_base_l (PPPP)"!(
    ComponentPose=@n20
    Alpha=1
    ConstraintWeights_0=0.0874
    ConstraintWeights_1:float=0.8646
    ConstraintWeights_2:float=0.0272
    ConstraintWeights_3:float=0.0208
   ) -> (Pose) {
    alphaBoolBlend {blendOption=Linear}
    boneToModify {boneName=dyn_pant_bttm_a_base_l}
    constraintSetup [
     {targetBone {boneName=calf_twist_01_l}}
     {targetBone {boneName=foot_l}}
     {targetBone {boneName=ball_l}}
     {targetBone {boneName=calf_twist_02_l}}
    ]
    constraintWeights [1,1,1,1]
   }
   n22 = "Constraint - dyn_pant_bttm_f_base_l (PP)"!(ComponentPose=@n21, Alpha=1, ConstraintWeights_0=0.1142, ConstraintWeights_1:float=0.8858) -> (Pose) {
    alphaBoolBlend {blendOption=Linear}
    boneToModify {boneName=dyn_pant_bttm_f_base_l}
    constraintSetup [{targetBone {boneName=calf_twist_01_l}}, {targetBone {boneName=foot_l}}]
    constraintWeights [1,1]
   }
   n23 = "Constraint - dyn_pant_bttm_d_base_l (PPP)"!(ComponentPose=@n22, Alpha=1, ConstraintWeights_0=0.5198, ConstraintWeights_1:float=0.4437, ConstraintWeights_2:float=0.0364) -> (Pose) {
    alphaBoolBlend {blendOption=Linear}
    boneToModify {boneName=dyn_pant_bttm_d_base_l}
    constraintSetup [{targetBone {boneName=calf_twist_01_l}}, {targetBone {boneName=foot_l}}, {targetBone {boneName=calf_twist_02_l}}]
    constraintWeights [1,1,1]
   }
   n24 = "Constraint - dyn_pant_bttm_c_base_l (PPP)"!(ComponentPose=@n23, Alpha=1, ConstraintWeights_0=0.3045, ConstraintWeights_1:float=0.6626, ConstraintWeights_2:float=0.0329) -> (Pose) {
    alphaBoolBlend {blendOption=Linear}
    boneToModify {boneName=dyn_pant_bttm_c_base_l}
    constraintSetup [{targetBone {boneName=calf_twist_01_l}}, {targetBone {boneName=foot_l}}, {targetBone {boneName=calf_twist_02_l}}]
    constraintWeights [1,1,1]
   }
   n25 = "Constraint - dyn_pant_bttm_h_base_l (PPP)"!(ComponentPose=@n24, Alpha=1, ConstraintWeights_0=0.3138, ConstraintWeights_1:float=0.6592, ConstraintWeights_2:float=0.0271) -> (Pose) {
    alphaBoolBlend {blendOption=Linear}
    boneToModify {boneName=dyn_pant_bttm_h_base_l}
    constraintSetup [{targetBone {boneName=calf_twist_01_l}}, {targetBone {boneName=foot_l}}, {targetBone {boneName=calf_twist_02_l}}]
    constraintWeights [1,1,1]
   }
   n26 = "Constraint - dyn_pant_bttm_b_base_l (PPPP)"!(
    ComponentPose=@n25
    Alpha=1
    ConstraintWeights_0=0.084
    ConstraintWeights_1:float=0.8646
    ConstraintWeights_2:float=0.0331
    ConstraintWeights_3:float=0.0183
   ) -> (Pose) {
    alphaBoolBlend {blendOption=Linear}
    boneToModify {boneName=dyn_pant_bttm_b_base_l}
    constraintSetup [
     {targetBone {boneName=calf_twist_01_l}}
     {targetBone {boneName=foot_l}}
     {targetBone {boneName=ball_l}}
     {targetBone {boneName=calf_twist_02_l}}
    ]
    constraintWeights [1,1,1,1]
   }
   n27 = "Constraint - dyn_pant_bttm_g_base_l (PP)"!(ComponentPose=@n26, Alpha=1, ConstraintWeights_0=0.1987, ConstraintWeights_1:float=0.8013) -> (Pose) {
    alphaBoolBlend {blendOption=Linear}
    boneToModify {boneName=dyn_pant_bttm_g_base_l}
    constraintSetup [{targetBone {boneName=calf_twist_01_l}}, {targetBone {boneName=foot_l}}]
    constraintWeights [1,1]
   }
   n28 = "Constraint - dyn_pant_bttm_e_base_l (PP)"!(ComponentPose=@n27, Alpha=1, ConstraintWeights_0=0.198, ConstraintWeights_1:float=0.802) -> (Pose) {
    alphaBoolBlend {blendOption=Linear}
    boneToModify {boneName=dyn_pant_bttm_e_base_l}
    constraintSetup [{targetBone {boneName=calf_twist_01_l}}, {targetBone {boneName=foot_l}}]
    constraintWeights [1,1]
   }
   n29 = "Constraint - dyn_pant_bttm_a_base_r (PPPP)"!(
    ComponentPose=@n28
    Alpha=1
    ConstraintWeights_0=0.0143
    ConstraintWeights_1:float=0.0826
    ConstraintWeights_2:float=0.2143
    ConstraintWeights_3:float=0.6888
   ) -> (Pose) {
    alphaBoolBlend {blendOption=Linear}
    boneToModify {boneName=dyn_pant_bttm_a_base_r}
    constraintSetup [
     {targetBone {boneName=ball_r}}
     {targetBone {boneName=calf_twist_02_r}}
     {targetBone {boneName=calf_twist_01_r}}
     {targetBone {boneName=foot_r}}
    ]
    constraintWeights [1,1,1,1]
   }
   n30 = "Constraint - dyn_pant_bttm_b_base_r (PPP)"!(ComponentPose=@n29, Alpha=1, ConstraintWeights_0=0.0065, ConstraintWeights_1:float=0.27, ConstraintWeights_2:float=0.7234) -> (Pose) {
    alphaBoolBlend {blendOption=Linear}
    boneToModify {boneName=dyn_pant_bttm_b_base_r}
    constraintSetup [{targetBone {boneName=calf_twist_02_r}}, {targetBone {boneName=calf_twist_01_r}}, {targetBone {boneName=foot_r}}]
    constraintWeights [1,1,1]
   }
   n31 = "Constraint - dyn_pant_bttm_c_base_r (PPP)"!(ComponentPose=@n30, Alpha=1, ConstraintWeights_0=0.021, ConstraintWeights_1:float=0.2481, ConstraintWeights_2:float=0.7309) -> (Pose) {
    alphaBoolBlend {blendOption=Linear}
    boneToModify {boneName=dyn_pant_bttm_c_base_r}
    constraintSetup [{targetBone {boneName=calf_twist_02_r}}, {targetBone {boneName=calf_twist_01_r}}, {targetBone {boneName=foot_r}}]
    constraintWeights [1,1,1]
   }
   n32 = "Constraint - dyn_pant_bttm_d_base_r (PPP)"!(ComponentPose=@n31, Alpha=1, ConstraintWeights_0=0.0005, ConstraintWeights_1:float=0.2031, ConstraintWeights_2:float=0.7964) -> (Pose) {
    alphaBoolBlend {blendOption=Linear}
    boneToModify {boneName=dyn_pant_bttm_d_base_r}
    constraintSetup [{targetBone {boneName=calf_twist_02_r}}, {targetBone {boneName=calf_twist_01_r}}, {targetBone {boneName=foot_r}}]
    constraintWeights [1,1,1]
   }
   n33 = "Constraint - dyn_pant_bttm_e_base_r (PPP)"!(ComponentPose=@n32, Alpha=1, ConstraintWeights_0=0.0344, ConstraintWeights_1:float=0.5157, ConstraintWeights_2:float=0.45) -> (Pose) {
    alphaBoolBlend {blendOption=Linear}
    boneToModify {boneName=dyn_pant_bttm_e_base_r}
    constraintSetup [{targetBone {boneName=calf_twist_02_r}}, {targetBone {boneName=calf_twist_01_r}}, {targetBone {boneName=foot_r}}]
    constraintWeights [1,1,1]
   }
   n34 = "Constraint - dyn_pant_bttm_f_base_r (PP)"!(ComponentPose=@n33, Alpha=1, ConstraintWeights_0=0.2445, ConstraintWeights_1:float=0.7555) -> (Pose) {
    alphaBoolBlend {blendOption=Linear}
    boneToModify {boneName=dyn_pant_bttm_f_base_r}
    constraintSetup [{targetBone {boneName=calf_twist_01_r}}, {targetBone {boneName=foot_r}}]
    constraintWeights [1,1]
   }
   n35 = "Transform (Modify) Bone - Bone: dyn_strap_a_base_l"(ComponentPose=@n34, Alpha=Get thigh_rot_l(), Translation="-2,0,0", Rotation="0,90,0", Scale="1,1,1") {
    rotationMode=BMM_Additive
    rotationSpace=BCS_BoneSpace
    scaleMode=BMM_Additive
    scaleSpace=BCS_BoneSpace
    translationMode=BMM_Additive
    translationSpace=BCS_BoneSpace
    alphaBoolBlend {blendOption=Linear}
    alphaScaleBiasClamp {bMapRange=true, inRange {max=0, min=-100}, outRange {max=0, min=1}}
    boneToModify {boneName=dyn_strap_a_base_l}
    rotation {yaw=90}
    translation {x=-2}
   }
   n36 = "Transform (Modify) Bone - Bone: dyn_strap_a_base_r"(ComponentPose=@n35, Alpha=Get thigh_rot_r(), Translation="2,0,0", Rotation="0,100,0", Scale="1,1,1") {
    rotationMode=BMM_Additive
    rotationSpace=BCS_BoneSpace
    scaleMode=BMM_Additive
    scaleSpace=BCS_BoneSpace
    translationMode=BMM_Additive
    translationSpace=BCS_BoneSpace
    alphaBoolBlend {blendOption=Linear}
    alphaScaleBiasClamp {bMapRange=true, inRange {max=0, min=-100}, outRange {max=0, min=1}}
    boneToModify {boneName=dyn_strap_a_base_r}
    rotation {yaw=100}
    translation {x=2}
   }
   n37 = Look At - Bone: dyn_pack_aim_l(ComponentPose=@n36, Alpha=1) {
    bUseLookUpAxis=true
    alphaBoolBlend {blendOption=Linear}
    boneToModify {boneName=dyn_pack_aim_l}
    lookAtLocation {x=0}
    lookAtTarget {boneReference {boneName=upperarm_l}}
    lookAt_Axis {axis=<1 0>}
    lookUp_Axis {axis {x=0, z=1}}
   }
   n38 = "Constraint - dyn_pack_base_l (T)"(ComponentPose=@n37, Alpha=0, ConstraintWeights_0=1) {
    alphaBoolBlend {blendOption=Linear}
    boneToModify {boneName=dyn_pack_base_l}
    constraintSetup [{transformType=Translation, targetBone {boneName=upperarm_l}}]
    constraintWeights [1]
   }
   n39 = Look At - Bone: dyn_pack_aim_r(ComponentPose=@n38, Alpha=1) {
    bUseLookUpAxis=true
    alphaBoolBlend {blendOption=Linear}
    boneToModify {boneName=dyn_pack_aim_r}
    lookAtLocation {x=0}
    lookAtTarget {boneReference {boneName=upperarm_r}}
    lookAt_Axis {axis=<-1 0>}
    lookUp_Axis {axis {x=0, z=1}}
   }
   n40 = "Constraint - dyn_pack_base_r (T)"(ComponentPose=@n39, Alpha=0, ConstraintWeights_0=1) {
    alphaBoolBlend {blendOption=Linear}
    boneToModify {boneName=dyn_pack_base_r}
    constraintSetup [{transformType=Translation, targetBone {boneName=upperarm_r}}]
    constraintWeights [1]
   }
   n41 = "Constraint - dyn_base_latt_base_l (PPPP)"!(
    ComponentPose=@n40
    Alpha=1
    ConstraintWeights_0=0.2
    ConstraintWeights_1:float=0.35
    ConstraintWeights_2:float=0.3
    ConstraintWeights_3:float=0.15
   ) -> (Pose) {
    alphaBoolBlend {blendOption=Linear}
    boneToModify {boneName=dyn_base_latt_base_l}
    constraintSetup [{targetBone {boneName=clavicle_l}}, {targetBone {boneName=spine_04}}, {targetBone {boneName=spine_03}}, {targetBone {boneName=spine_02}}]
    constraintWeights [1,1,1,1]
   }
   n42 = "Constraint - dyn_base_ratt_base_r (PPPP)"!(
    ComponentPose=@n41
    Alpha=1
    ConstraintWeights_0=0.2
    ConstraintWeights_1:float=0.35
    ConstraintWeights_2:float=0.3
    ConstraintWeights_3:float=0.15
   ) -> (Pose) {
    alphaBoolBlend {blendOption=Linear}
    boneToModify {boneName=dyn_base_ratt_base_r}
    constraintSetup [{targetBone {boneName=clavicle_r}}, {targetBone {boneName=spine_04}}, {targetBone {boneName=spine_03}}, {targetBone {boneName=spine_02}}]
    constraintWeights [1,1,1,1]
   }
   n43 = "Constraint - dyn_shldr_bk_base_l (PPPP)"!(
    ComponentPose=@n42
    Alpha=1
    ConstraintWeights_0=0.5
    ConstraintWeights_1:float=0.3
    ConstraintWeights_2:float=0.1
    ConstraintWeights_3:float=0.1
   ) -> (Pose) {
    alphaBoolBlend {blendOption=Linear}
    boneToModify {boneName=dyn_shldr_bk_base_l}
    constraintSetup [
     {targetBone {boneName=upperarm_twist_01_l}}
     {targetBone {boneName=clavicle_l}}
     {targetBone {boneName=spine_04}}
     {targetBone {boneName=spine_03}}
    ]
    constraintWeights [1,1,1,1]
   }
   n44 = "Constraint - dyn_shldr_bk_base_r (PPPP)"!(
    ComponentPose=@n43
    Alpha=1
    ConstraintWeights_0=0.5
    ConstraintWeights_1:float=0.3
    ConstraintWeights_2:float=0.1
    ConstraintWeights_3:float=0.1
   ) -> (Pose) {
    alphaBoolBlend {blendOption=Linear}
    boneToModify {boneName=dyn_shldr_bk_base_r}
    constraintSetup [
     {targetBone {boneName=upperarm_twist_01_r}}
     {targetBone {boneName=clavicle_r}}
     {targetBone {boneName=spine_04}}
     {targetBone {boneName=spine_03}}
    ]
    constraintWeights [1,1,1,1]
   }
   n45 = Get l_arm_fwd() -> (l_arm_fwd:double)
   n46 = "Transform (Modify) Bone - Bone: dyn_pack_l"(ComponentPose=@n44, Alpha=@n45, Translation="-15,0,0", Rotation="0,0,0", Scale="1,1,1") {
    rotationMode=BMM_Additive
    rotationSpace=BCS_BoneSpace
    scaleMode=BMM_Additive
    scaleSpace=BCS_BoneSpace
    translationMode=BMM_Additive
    translationSpace=BCS_BoneSpace
    alphaBoolBlend {blendOption=Linear}
    alphaScaleBiasClamp {inRange {max=0, min=-100}, outRange {max=0, min=1}}
    boneToModify {boneName=dyn_pack_l}
    translation {x=-15}
   }
   n47 = "Transform (Modify) Bone - Bone: dyn_bicep_fr_l"(ComponentPose=@n46, Alpha=@n45, Translation="0,-10,0", Rotation="0,0,0", Scale="1,1,1") {
    rotationMode=BMM_Additive
    rotationSpace=BCS_BoneSpace
    scaleMode=BMM_Additive
    scaleSpace=BCS_BoneSpace
    translationMode=BMM_Additive
    translationSpace=BCS_BoneSpace
    alphaBoolBlend {blendOption=Linear}
    alphaScaleBiasClamp {inRange {max=0, min=-100}, outRange {max=0, min=1}}
    boneToModify {boneName=dyn_bicep_fr_l}
    translation {y=-10}
   }
   n48 = "Transform (Modify) Bone - Bone: dyn_shldr_bk_l"(ComponentPose=@n47, Alpha=@n45, Translation="0,-3,0", Rotation="0,0,0", Scale="1,1,1") {
    rotationMode=BMM_Additive
    rotationSpace=BCS_BoneSpace
    scaleMode=BMM_Additive
    scaleSpace=BCS_BoneSpace
    translationMode=BMM_Additive
    translationSpace=BCS_BoneSpace
    alphaBoolBlend {blendOption=Linear}
    alphaScaleBiasClamp {inRange {max=0, min=-100}, outRange {max=0, min=1}}
    boneToModify {boneName=dyn_shldr_bk_l}
    translation {y=-3}
   }
   n49 = Get l_arm_up() -> (l_arm_up:double)
   n50 = "Transform (Modify) Bone - Bone: dyn_pack_l"(ComponentPose=@n48, Alpha=@n49, Translation="10,0,0", Rotation="0,0,0", Scale="1,1,1") {
    rotationMode=BMM_Additive
    rotationSpace=BCS_BoneSpace
    scaleMode=BMM_Additive
    scaleSpace=BCS_BoneSpace
    translationMode=BMM_Additive
    translationSpace=BCS_BoneSpace
    alphaBoolBlend {blendOption=Linear}
    alphaScaleBiasClamp {inRange {max=0, min=-100}, outRange {max=0, min=1}}
    boneToModify {boneName=dyn_pack_l}
    translation {x=10}
   }
   n51 = "Transform (Modify) Bone - Bone: dyn_bicep_fr_l"(ComponentPose=@n50, Alpha=@n49, Translation="0,6,0", Rotation="0,0,0", Scale="1,1,1") {
    rotationMode=BMM_Additive
    rotationSpace=BCS_BoneSpace
    scaleMode=BMM_Additive
    scaleSpace=BCS_BoneSpace
    translationMode=BMM_Additive
    translationSpace=BCS_BoneSpace
    alphaBoolBlend {blendOption=Linear}
    alphaScaleBiasClamp {inRange {max=0, min=-100}, outRange {max=0, min=1}}
    boneToModify {boneName=dyn_bicep_fr_l}
    translation {y=6}
   }
   n52 = Get l_arm_dn() -> (l_arm_dn:double)
   n53 = "Transform (Modify) Bone - Bone: dyn_latt_l"(ComponentPose=@n51, Alpha=@n52, Translation="0,0,7", Rotation="0,0,0", Scale="1,1,1") {
    rotationMode=BMM_Additive
    rotationSpace=BCS_BoneSpace
    scaleMode=BMM_Additive
    scaleSpace=BCS_BoneSpace
    translationMode=BMM_Additive
    translationSpace=BCS_BoneSpace
    alphaBoolBlend {blendOption=Linear}
    alphaScaleBiasClamp {inRange {max=0, min=-100}, outRange {max=0, min=1}}
    boneToModify {boneName=dyn_latt_l}
    translation {z=7}
   }
   n54 = "Transform (Modify) Bone - Bone: upperarm_twist_02_l"(ComponentPose=@n53, Alpha=@n52, Translation="0,0,0", Rotation="0,0,0", Scale="1,1,0.8") {
    rotationMode=BMM_Additive
    rotationSpace=BCS_BoneSpace
    scaleMode=BMM_Additive
    scaleSpace=BCS_BoneSpace
    translationMode=BMM_Additive
    translationSpace=BCS_BoneSpace
    alphaBoolBlend {blendOption=Linear}
    alphaScaleBiasClamp {inRange {max=0, min=-100}, outRange {max=0, min=1}}
    boneToModify {boneName=upperarm_twist_02_l}
    scale {z=0.8}
   }
   n55 = Get r_arm_fwd() -> (r_arm_fwd:double)
   n56 = "Transform (Modify) Bone - Bone: dyn_pack_r"(ComponentPose=@n54, Alpha=@n55, Translation="15,0,0", Rotation="0,0,0", Scale="1,1,1") {
    rotationMode=BMM_Additive
    rotationSpace=BCS_BoneSpace
    scaleMode=BMM_Additive
    scaleSpace=BCS_BoneSpace
    translationMode=BMM_Additive
    translationSpace=BCS_BoneSpace
    alphaBoolBlend {blendOption=Linear}
    alphaScaleBiasClamp {inRange {max=0, min=-100}, outRange {max=0, min=1}}
    boneToModify {boneName=dyn_pack_r}
    translation {x=15}
   }
   n57 = "Transform (Modify) Bone - Bone: dyn_bicep_fr_r"(ComponentPose=@n56, Alpha=@n55, Translation="0,10,0", Rotation="0,0,0", Scale="1,1,1") {
    rotationMode=BMM_Additive
    rotationSpace=BCS_BoneSpace
    scaleMode=BMM_Additive
    scaleSpace=BCS_BoneSpace
    translationMode=BMM_Additive
    translationSpace=BCS_BoneSpace
    alphaBoolBlend {blendOption=Linear}
    alphaScaleBiasClamp {inRange {max=0, min=-100}, outRange {max=0, min=1}}
    boneToModify {boneName=dyn_bicep_fr_r}
    translation {y=10}
   }
   n58 = "Transform (Modify) Bone - Bone: dyn_shldr_bk_r"(ComponentPose=@n57, Alpha=@n55, Translation="0,3,0", Rotation="0,0,0", Scale="1,1,1") {
    rotationMode=BMM_Additive
    rotationSpace=BCS_BoneSpace
    scaleMode=BMM_Additive
    scaleSpace=BCS_BoneSpace
    translationMode=BMM_Additive
    translationSpace=BCS_BoneSpace
    alphaBoolBlend {blendOption=Linear}
    alphaScaleBiasClamp {inRange {max=0, min=-100}, outRange {max=0, min=1}}
    boneToModify {boneName=dyn_shldr_bk_r}
    translation {y=3}
   }
   n59 = Get r_arm_up() -> (r_arm_up:double)
   n60 = "Transform (Modify) Bone - Bone: dyn_pack_r"(ComponentPose=@n58, Alpha=@n59, Translation="-10,0,0", Rotation="0,0,0", Scale="1,1,1") {
    rotationMode=BMM_Additive
    rotationSpace=BCS_BoneSpace
    scaleMode=BMM_Additive
    scaleSpace=BCS_BoneSpace
    translationMode=BMM_Additive
    translationSpace=BCS_BoneSpace
    alphaBoolBlend {blendOption=Linear}
    alphaScaleBiasClamp {inRange {max=0, min=-100}, outRange {max=0, min=1}}
    boneToModify {boneName=dyn_pack_r}
    translation {x=-10}
   }
   n61 = "Transform (Modify) Bone - Bone: dyn_bicep_fr_r"(ComponentPose=@n60, Alpha=@n59, Translation="0,-6,0", Rotation="0,0,0", Scale="1,1,1") {
    rotationMode=BMM_Additive
    rotationSpace=BCS_BoneSpace
    scaleMode=BMM_Additive
    scaleSpace=BCS_BoneSpace
    translationMode=BMM_Additive
    translationSpace=BCS_BoneSpace
    alphaBoolBlend {blendOption=Linear}
    alphaScaleBiasClamp {inRange {max=0, min=-100}, outRange {max=0, min=1}}
    boneToModify {boneName=dyn_bicep_fr_r}
    translation {y=-6}
   }
   n62 = Get r_arm_dn() -> (r_arm_dn:double)
   n63 = "Transform (Modify) Bone - Bone: dyn_ratt_r"(ComponentPose=@n61, Alpha=@n62, Translation="0,0,-7", Rotation="0,0,0", Scale="1,1,1") {
    rotationMode=BMM_Additive
    rotationSpace=BCS_BoneSpace
    scaleMode=BMM_Additive
    scaleSpace=BCS_BoneSpace
    translationMode=BMM_Additive
    translationSpace=BCS_BoneSpace
    alphaBoolBlend {blendOption=Linear}
    alphaScaleBiasClamp {inRange {max=0, min=-100}, outRange {max=0, min=1}}
    boneToModify {boneName=dyn_ratt_r}
    translation {z=-7}
   }
   n64 = "Transform (Modify) Bone - Bone: upperarm_twist_02_r"(ComponentPose=@n63, Alpha=@n62, Translation="0,0,0", Rotation="0,0,0", Scale="1,1,0.8") {
    rotationMode=BMM_Additive
    rotationSpace=BCS_BoneSpace
    scaleMode=BMM_Additive
    scaleSpace=BCS_BoneSpace
    translationMode=BMM_Additive
    translationSpace=BCS_BoneSpace
    alphaBoolBlend {blendOption=Linear}
    alphaScaleBiasClamp {inRange {max=0, min=-100}, outRange {max=0, min=1}}
    boneToModify {boneName=upperarm_twist_02_r}
    scale {z=0.8}
   }
   n65 = Get r_pistonAimUp_alpha() -> (r_pistonAimUp_alpha:double)
   n66 = "Transform (Modify) Bone - Bone: upperarm_twist_01_r"(ComponentPose=@n64, Alpha=@n65, Translation="3,0,0", Rotation="-30,0,0", Scale="1,1,1") {
    rotationMode=BMM_Additive
    rotationSpace=BCS_BoneSpace
    scaleMode=BMM_Additive
    scaleSpace=BCS_BoneSpace
    translationMode=BMM_Additive
    translationSpace=BCS_BoneSpace
    alphaBoolBlend {blendOption=Linear}
    alphaScaleBiasClamp {inRange {max=0, min=-100}, outRange {max=0, min=1}}
    boneToModify {boneName=upperarm_twist_01_r}
    rotation {pitch=-30}
    translation {x=3}
   }
   n67 = "Transform (Modify) Bone - Bone: upperarm_twist_02_r"(ComponentPose=@n66, Alpha=@n65, Translation="0,0,-5", Rotation="0,0,0", Scale="1,1,1") {
    rotationMode=BMM_Additive
    rotationSpace=BCS_BoneSpace
    scaleMode=BMM_Additive
    scaleSpace=BCS_BoneSpace
    translationMode=BMM_Additive
    translationSpace=BCS_BoneSpace
    alphaBoolBlend {blendOption=Linear}
    alphaScaleBiasClamp {inRange {max=0, min=-100}, outRange {max=0, min=1}}
    boneToModify {boneName=upperarm_twist_02_r}
    translation {z=-5}
   }
   n68 = "Transform (Modify) Bone - Bone: dyn_bicep_bk_r"(ComponentPose=@n67, Alpha=@n65, Translation="0,-2,0", Rotation="0,0,0", Scale="1,1,1") {
    rotationMode=BMM_Additive
    rotationSpace=BCS_BoneSpace
    scaleMode=BMM_Additive
    scaleSpace=BCS_BoneSpace
    translationMode=BMM_Additive
    translationSpace=BCS_BoneSpace
    alphaBoolBlend {blendOption=Linear}
    alphaScaleBiasClamp {inRange {max=0, min=-100}, outRange {max=0, min=1}}
    boneToModify {boneName=dyn_bicep_bk_r}
    translation {y=-2}
   }
   n69 = Get crouchingAlpha() -> (crouchingAlpha:double)
   n70 = "Transform (Modify) Bone - Bone: dyn_strap_a_base_l"(ComponentPose=@n68, Alpha=@n69, Translation="0,0,0", Rotation=Get LoinClothRotator_01() -> (LoinClothRotator_01:Rotator), Scale="1,1,1") {
    rotationMode=BMM_Additive
    rotationSpace=BCS_WorldSpace
    translationSpace=BCS_BoneSpace
    alphaBoolBlend {blendOption=Linear}
    alphaScaleBiasClamp {bInterpResult=true, bMapRange=true, interpSpeedDecreasing=2, interpSpeedIncreasing=2, outRange {max=0, min=1}}
    boneToModify {boneName=dyn_strap_a_base_l}
   }
   n71 = "Transform (Modify) Bone - Bone: dyn_strap_a_base_r"(ComponentPose=@n70, Alpha=@n69, Translation="0,0,0", Rotation=Get LoinClothRotator_02() -> (LoinClothRotator_02:Rotator), Scale="1,1,1") {
    rotationMode=BMM_Additive
    rotationSpace=BCS_WorldSpace
    translationSpace=BCS_BoneSpace
    alphaBoolBlend {blendOption=Linear}
    alphaScaleBiasClamp {bInterpResult=true, bMapRange=true, interpSpeedDecreasing=2, interpSpeedIncreasing=2, outRange {max=0, min=1}}
    boneToModify {boneName=dyn_strap_a_base_r}
   }
   n72 = "Transform (Modify) Bone - Bone: dyn_arm_strap_aa_base_l"(ComponentPose=@n71, Alpha=0.7, Translation="0,0,0", Rotation=Get LoinClothRotator_01(), Scale="1,1,1") {
    rotationMode=BMM_Additive
    rotationSpace=BCS_WorldSpace
    translationSpace=BCS_BoneSpace
    alphaBoolBlend {blendOption=Linear}
    alphaScaleBiasClamp {bInterpResult=true, interpSpeedDecreasing=2, interpSpeedIncreasing=2}
    boneToModify {boneName=dyn_arm_strap_aa_base_l}
   }
   n73 = "Transform (Modify) Bone - Bone: dyn_arm_strap_ba_base_l"(ComponentPose=@n72, Alpha=0.7, Translation="0,0,0", Rotation=Get LoinClothRotator_02(), Scale="1,1,1") {
    rotationMode=BMM_Additive
    rotationSpace=BCS_WorldSpace
    translationSpace=BCS_BoneSpace
    alphaBoolBlend {blendOption=Linear}
    alphaScaleBiasClamp {bInterpResult=true, interpSpeedDecreasing=2, interpSpeedIncreasing=2}
    boneToModify {boneName=dyn_arm_strap_ba_base_l}
   }
   n74 = "Transform (Modify) Bone - Bone: dyn_pants_scrap_base_l"(ComponentPose=@n73, Alpha=0.5, Translation="0,0,0", Rotation=Get LoinClothRotator_01(), Scale="1,1,1") {
    rotationMode=BMM_Additive
    rotationSpace=BCS_WorldSpace
    translationSpace=BCS_BoneSpace
    alphaBoolBlend {blendOption=Linear}
    alphaScaleBiasClamp {bInterpResult=true, interpSpeedDecreasing=2, interpSpeedIncreasing=2}
    boneToModify {boneName=dyn_pants_scrap_base_l}
   }
   n75 = "Transform (Modify) Bone - Bone: dyn_pants_scrap_base_r"(ComponentPose=@n74, Alpha=0.5, Translation="0,0,0", Rotation=Get LoinClothRotator_02(), Scale="1,1,1") {
    rotationMode=BMM_Additive
    rotationSpace=BCS_WorldSpace
    translationSpace=BCS_BoneSpace
    alphaBoolBlend {blendOption=Linear}
    alphaScaleBiasClamp {bInterpResult=true, interpSpeedDecreasing=2, interpSpeedIncreasing=2}
    boneToModify {boneName=dyn_pants_scrap_base_r}
   }
   n76 = Get isInFrontEndAlpha() -> (isInFrontEndAlpha:double)
   n77 = "Transform (Modify) Bone - Bone: dyn_root_simspace"(ComponentPose=@n75, Alpha=@n76, Translation="0,0,0", Rotation=Get sim_space_rotation() -> (sim_space_rotation:Rotator), Scale="1,1,1") {
    rotationMode=BMM_Replace
    rotationSpace=BCS_WorldSpace
    translationSpace=BCS_BoneSpace
    alphaBoolBlend {blendOption=Linear}
    alphaScaleBiasClamp {bInterpResult=true, bMapRange=true, interpSpeedDecreasing=2, interpSpeedIncreasing=2, outRange {max=0, min=1}}
    boneToModify {boneName=dyn_root_simspace}
   }
   n78 = "Constraint - dyn_root_simspace (P)"(ComponentPose=@n77, Alpha=@n76, ConstraintWeights_0=1) {
    alphaBoolBlend {blendOption=Linear}
    boneToModify {boneName=dyn_root_simspace}
    constraintSetup [{targetBone {boneName=pelvis}}]
    constraintWeights [1]
   }
   n79 = Get GravityOverride_tessel() -> (GravityOverride_tessel:Vector)
   n80 = RigidBody!(ComponentPose=@n78, Alpha, OverrideWorldGravity:Vector=@n79, ExternalForce) -> (Pose) {
    bClampLinearTranslationLimitToRefPose=true
    bOverrideWorldGravity=true
    overridePhysicsAsset=/Game/Characters/Player/Male/Medium/Bodies/M_MED_Skirmish_Alt/Meshes/M_MED_Skirmish_Shared_Physics
    simulationSpace=BaseBoneSpace
    alphaBoolBlend {blendOption=Linear}
    baseBoneRef {boneName=dyn_root_simspace}
    componentLinearAccScale=<0.3 0.3 0.3>
    componentLinearVelScale=<0.3 0.3 0.3>
    overrideWorldGravity {z=-980}
   }
   n81 = RigidBody(ComponentPose=@n80, OverrideWorldGravity=@n79) {
    bClampLinearTranslationLimitToRefPose=true
    bOverrideWorldGravity=true
    overridePhysicsAsset=&/M_MED_Skirmish_Alt_Physics
    simulationSpace=BaseBoneSpace
    alphaBoolBlend {blendOption=Linear}
    baseBoneRef {boneName=dyn_root_simspace}
    componentLinearAccScale=<0.3 0.3 0.3>
    componentLinearVelScale=<0.3 0.3 0.3>
    overrideWorldGravity {z=-980}
   }
   n82 = RigidBody!(ComponentPose=@n81, Alpha, ExternalForce) -> (Pose) {
    bClampLinearTranslationLimitToRefPose=true
    bOverrideWorldGravity=true
    overridePhysicsAsset=&/M_MED_Skirmish_Alt_Belt_Physics
    simulationSpace=BaseBoneSpace
    alphaBoolBlend {blendOption=Linear}
    baseBoneRef {boneName=dyn_root_simspace}
    componentLinearAccScale=<0.3 0.3 0.3>
    componentLinearVelScale=<0.3 0.3 0.3>
    overrideWorldGravity {z=-980}
   }
   Output Pose(Result=Component To Local(ComponentPose=@n82)) {layerGroup=None}
  ]}
 EventGraph {nodes [
   Event BlueprintUpdateAnimation()
   > n1 = Sequence!(execute) -> (then_0, then_1, then_2:Exec, then_3:Exec, then_4:Exec, then_5:Exec, then_6:Exec)
   n2 = Sequence(execute=@n1.then_6)
   n3 = Sequence(execute=@n2.then_3)
   n4 = Sequence(execute=@n3.then_5)
   n5 = Branch(execute=@n4.then_2, Condition=Get bIsInFrontEnd())
   > Set isInFrontEndAlpha(isInFrontEndAlpha:double=1) -> (Output_Get:double)
   Set isInFrontEndAlpha(execute=@n5.else, isInFrontEndAlpha=0)
   n6 = Branch(execute=@n4.then_1, Condition=Get bIsCrouching())
   > Set crouchingAlpha(crouchingAlpha:double=1) -> (Output_Get:double)
   Set crouchingAlpha(execute=@n6.else, crouchingAlpha=0)
   n7 = ComplexSimSpaceFunction DEPRECATED!(
    execute=@n4.then_0
    CharacterPartAnimInstance:Custom Character Part Anim Instance Object Reference=Self-Reference()
    InterpSpeedMax:double=9
    PawnSpeedMax:double=20
    RInterpErrorTolerance:double=1
    DeltaTime:double=0
    DebugDraw:bool=false
    sim_space_rotation_in:Rotator=Get sim_space_rotation()
    inComponentSpaceIn:bool=Get InComponentSpace() -> (InComponentSpace:bool)
    inComponentSpaceLastFrameIn:bool=Get InComponentSpaceLastFrame() -> (InComponentSpaceLastFrame:bool)
    lockedOnSimSpaceIn:bool=Get LockedOnSimSpace() -> (LockedOnSimSpace:bool)
    InterpSpeedIn:double=Get InterpSpeed() -> (InterpSpeed:double)
   ) -> (
    then
    sim_space_rotation:Rotator
    inComponentSpaceOut:bool
    inComponentSpaceLastFrameOut:bool
    lockedOnSimSpaceOut:bool
    InterpSpeedOut:double
   )
   > Set sim_space_rotation(sim_space_rotation:Rotator=@n7) -> (Output_Get:Rotator)
   > Set InComponentSpace(InComponentSpace:bool=@n7.inComponentSpaceOut) -> (Output_Get:bool)
   > Set InComponentSpaceLastFrame(InComponentSpaceLastFrame:bool=@n7.inComponentSpaceLastFrameOut) -> (Output_Get:bool)
   > Set LockedOnSimSpace(LockedOnSimSpace:bool=@n7.lockedOnSimSpaceOut) -> (Output_Get:bool)
   > Set InterpSpeed(InterpSpeed:double=@n7.InterpSpeedOut) -> (Output_Get:double)
   n8 = Gravity Override State Function DEPRECATED(
    execute=@n3.then_3
    CharacterPartAnimInstance=Self-Reference()
    GameTimeInSeconds=GetGameTimeInSeconds()
    GravityOverrideParams=Get Gravity Override Params()
    DebugJointName=spine_05
   )
   > n9 = Set GravityOverride_tessel(GravityOverride_tessel:Vector=@n8) -> (Output_Get:Vector)
   n10 = Branch(execute=@n3.then_1, Condition=Get isPistol() -> (isPistol:bool))
   > n11 = Branch(Condition="float >= float"(A=Get r_arm_fwd(), B=0.8) {is=title K2Node_CallFunction})
   > Set r_pistonAimUp_alpha(r_pistonAimUp_alpha:double=MapRangeClamped(Value=Get r_arm_up(), InRangeA=0.5, InRangeB=1)) -> (Output_Get:double)
   Set r_pistonAimUp_alpha(execute=@n11.else, r_pistonAimUp_alpha=0)
   Set r_pistonAimUp_alpha(execute=@n10.else, r_pistonAimUp_alpha=0)
   Set LoinClothRotator_01(
    execute=@n3.then_2
    LoinClothRotator_01:Rotator=MakeWaveRotator!(Amplitudes:Rotator="75,0,8", Frequencies:Rotator="50,0,80", Offsets:Rotator="0,0,0", bUseNoise:bool=true, Min:Vector, Max:Vector="0,180, 0") -> (ReturnValue:Rotator)
   ) -> (Output_Get:Rotator)
   > Set LoinClothRotator_02(LoinClothRotator_02:Rotator=MakeWaveRotator(Amplitudes="75,0,8", Frequencies="50,0,80", Offsets="10,0,5", bUseNoise=true, Max="0,180, 0")) -> (Output_Get:Rotator)
   n12 = Cast To FortPlayerAnimInstance(execute=@n3.then_0, Object=GetAnimInstance(self=GetOwnerSkeletalMeshComponent()))
   > Set isPistol(isPistol="Equal (Enum)"(A:EFortWeaponCoreAnimation Enum=Get WeaponCoreAnim(self=@n12), B:EFortWeaponCoreAnimation Enum=Pistol))
   n13 = GetPartSkeletalMeshComponent()
   Set l_arm_fwd(
    execute=@n2.then_0
    l_arm_fwd:double=GetAlphaFromReferenceTransformDelta(
     SkeletalMeshComponent=@n13
     BoneName=lowerarm_l
     BaseName=dyn_arm_measure_l
     DeltaComponentType=TranslationY
     InputRange=(X=5,Y=25)
     TwistAxis="1,0,0"
    )
   ) -> (Output_Get:double)
   n14 = GetPartSkeletalMeshComponent()
   > Set r_arm_fwd(r_arm_fwd:double=GetAlphaFromReferenceTransformDelta(
     SkeletalMeshComponent=@n14
     BoneName=lowerarm_r
     BaseName=dyn_arm_measure_r
     DeltaComponentType=TranslationY
     InputRange=(X=-25,Y=-5)
     OutputRange=(X=1,Y=0)
     TwistAxis="1,0,0"
    )) -> (Output_Get:double)
   Set l_arm_up(
    execute=@n2.then_1
    l_arm_up:double=GetAlphaFromReferenceTransformDelta(
     SkeletalMeshComponent=@n13
     BoneName=lowerarm_l
     BaseName=dyn_arm_measure_l
     DeltaComponentType=TranslationZ
     InputRange=(X=10,Y=25)
     TwistAxis="1,0,0"
    )
   ) -> (Output_Get:double)
   > Set r_arm_up(r_arm_up:double=GetAlphaFromReferenceTransformDelta(
     SkeletalMeshComponent=@n14
     BoneName=lowerarm_r
     BaseName=dyn_arm_measure_r
     DeltaComponentType=TranslationZ
     InputRange=(X=-25,Y=-10)
     OutputRange=(X=1,Y=0)
     TwistAxis="1,0,0"
    )) -> (Output_Get:double)
   Set l_arm_dn(
    execute=@n2.then_2
    l_arm_dn:double=GetAlphaFromReferenceTransformDelta(
     SkeletalMeshComponent=@n13
     BoneName=lowerarm_l
     BaseName=dyn_arm_measure_l
     DeltaComponentType=TranslationZ
     InputRange=(X=-7,Y=0)
     OutputRange=(X=1,Y=0)
     TwistAxis="1,0,0"
    )
   ) -> (Output_Get:double)
   > Set r_arm_dn(r_arm_dn:double=GetAlphaFromReferenceTransformDelta(
     SkeletalMeshComponent=@n14
     BoneName=lowerarm_r
     BaseName=dyn_arm_measure_r
     DeltaComponentType=TranslationZ
     InputRange=(X=0,Y=7)
     TwistAxis="1,0,0"
    )) -> (Output_Get:double)
   n15 = GetDeltaTransformFromRefPose(self=GetOwnerSkeletalMeshComponent(), BoneName=lowerarm_l, BaseName=upperarm_l)
   Set elbow_rot_l(execute=@n1.then_0, elbow_rot_l:double=BreakRotator(InRot=BreakTransform(InTransform=@n15).Rotation).Yaw) -> (Output_Get:double)
   n16 = GetDeltaTransformFromRefPose(self=GetOwnerSkeletalMeshComponent(), BoneName=thigh_r, BaseName=pelvis)
   Set thigh_rot_r(execute=@n1.then_3, thigh_rot_r=BreakRotator(InRot=BreakTransform(InTransform=@n16).Rotation).Yaw)
   n17 = GetDeltaTransformFromRefPose(self=GetOwnerSkeletalMeshComponent(), BoneName=calf_r, BaseName=thigh_r)
   Set knee_rot_r(execute=@n1.then_5, knee_rot_r:double=BreakRotator(InRot=BreakTransform(InTransform=@n17).Rotation).Yaw) -> (Output_Get:double)
   n18 = GetDeltaTransformFromRefPose(self=GetOwnerSkeletalMeshComponent(), BoneName=calf_l, BaseName=thigh_l)
   Set knee_rot_l(execute=@n1.then_4, knee_rot_l:double=BreakRotator(InRot=BreakTransform(InTransform=@n18).Rotation).Yaw) -> (Output_Get:double)
   n19 = GetDeltaTransformFromRefPose(self=GetOwnerSkeletalMeshComponent(), BoneName=thigh_l, BaseName=pelvis)
   Set thigh_rot_l(execute=@n1.then_2, thigh_rot_l=BreakRotator(InRot=BreakTransform(InTransform=@n19).Rotation).Yaw)
   n20 = GetDeltaTransformFromRefPose(self=GetOwnerSkeletalMeshComponent(), BoneName=lowerarm_r, BaseName=upperarm_r)
   Set elbow_rot_r(execute=@n1.then_1, elbow_rot_r:double=BreakRotator(InRot=BreakTransform(InTransform=@n20).Rotation).Yaw) -> (Output_Get:double)
   TryGetPawnOwner()
   DrawVectorFromJointFunction!(
    execute
    SkelMeshComponent:Scene Component Object Reference
    JointName:Name=head
    LineEnd:Vector=@n9
    JointOffset:Vector="0, 0, 0"
    LineColor:Linear Color Structure=(R=1,G=0,B=0,A=1)
    Thickness:double=2
    Foreground:EDrawDebugSceneDepthPriorityGroup Enum=Foreground
   ) -> (then)
  ]}
]
variables [
 crouchingAlpha {type=number}
 elbow_rot_l {type=number}
 elbow_rot_r {type=number}
 Gravity Override Params {
  type=GravityOverrideParamsStruct
  default {
   emoteOrMelee {gravityOverride=<0 0 -15>, windAmplitude=<100 -200 100>, windFrequency=<7 7 7>, windOffset=<0 -70 0>}
   front_End {gravityOverride=<0 0 -15>, windAmplitude=<100 -200 100>, windFrequency=<7 7 7>, windOffset=<0 -70 0>}
   onGround_Moving {gravityOverride=<0 0 -15>, windAmplitude=<100 -200 100>, windFrequency=<7 7 7>, windOffset=<0 -70 0>}
   onGround_Standing {gravityOverride=<0 0 -15>, windAmplitude=<100 -200 100>, windFrequency=<7 7 7>, windOffset=<0 -70 0>}
  }
 }
 GravityOverride_tessel {type=Vector}
 InComponentSpace {type=boolean}
 InComponentSpaceLastFrame {type=boolean}
 InterpSpeed {type=number}
 isInFrontEndAlpha {type=number}
 isPistol {type=boolean}
 knee_rot_l {type=number}
 knee_rot_r {type=number}
 l_arm_dn {type=number}
 l_arm_fwd {type=number}
 l_arm_up {type=number}
 LockedOnSimSpace {type=boolean}
 LoinClothRotator_01 {type=Rotator}
 LoinClothRotator_02 {type=Rotator}
 r_arm_dn {type=number}
 r_arm_fwd {type=number}
 r_arm_up {type=number}
 r_pistonAimUp_alpha {type=number}
 sim_space_rotation {type=Rotator}
 thigh_rot_l {type=number}
 thigh_rot_r {type=number}
]
